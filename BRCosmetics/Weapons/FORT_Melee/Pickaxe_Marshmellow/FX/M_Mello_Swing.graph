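# Translucent unlit two-sided material for the Marshmellow pickaxe swing trail FX, shown as a panning rainbow streak tinted by particle color.
# A tiled panner mask modulates opacity and HDR emissive intensity, with pan speed driven by a dynamic parameter.
# Sine and power UV envelopes shape the streak, while a Swipe parameter offsets a gradient wipe, overridable by a debug switch.
# Opacity also uses particle alpha, view-angle fade and depth fade.
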Material /BRCosmetics/Weapons/FORT_Melee/Pickaxe_Marshmellow/FX/M_Mello_Swing
graph {
 properties {
  blendMode=BLEND_Translucent
  pixelDepthOffsetMode=PDOM_Legacy
  previewMesh=&/S_FX_MelloSwing_Swing01_Mesh
  shadingModel=MSM_Unlit
  twoSided=true
  usedShadingModels=Unlit
 }
 nodes [
  n1 = ParticleColor()
  n2 = Multiply(
   A=TextureCoordinate() {vTiling=-1}
   B=AppendVector(A=ScalarParameter() {defaultValue=2, parameterName=U_Scale}, B=ScalarParameter() {defaultValue=1, parameterName=V_Scale})
  )
  n3 = Multiply(A=Time(), B=DynamicParameter() {defaultValue=<rgba 1 0.5 1 1>, paramNames [PanningSpeed,Swipe,Param3,Param4]}.PanningSpeed)
  n4 = TextureSample(UVs=Panner(Coordinate=MaterialFunctionCall(XY=@n2) {materialFunction=&Swizzle}.YX, Time=@n3) {speedX=3}) {texture=/Game/Effects/Fort_Effects/Textures/Tile/Tests/T_Tile_Panner_Mask}
  n5 = LinearInterpolate(Alpha=@n4.R) {constA=0.2}
  n6 = Sine(None=ComponentMask(None=TextureCoordinate()) {r=true}) {period=2}
  n7 = ComponentMask(None=TextureCoordinate()) {g=true}
  n8 = ScalarParameter() {defaultValue=3, parameterName=Lead}
  n9 = OneMinus(None=@n7)
  n10 = ScalarParameter() {defaultValue=2, parameterName=End}
  n11 = Multiply(A=Multiply(A=Power(Base=@n7, Exp=@n8), B=Power(Base=@n9, Exp=@n10))) {constB=100}
  n12 = TextureCoordinate() {vTiling=0.25}
  n13 = ComponentMask(None=@n12) {r=true}
  n14 = Add(
   A=ComponentMask(None=@n12) {g=true}
   B=StaticSwitchParameter(
    True=ScalarParameter() {defaultValue=0.25, parameterName=DEBUG_Swipe}
    False=DynamicParameter() {defaultValue=<rgba 1 0.5 1 1>, paramNames [PanningSpeed,Swipe,Param3,Param4]}.Swipe
   ) {group=DEBUG, parameterName=DEBUG}
  )
  n15 = Multiply(
   A=Multiply(A=Multiply(A=@n5, B=@n6), B=Saturate(None=@n11))
   B=TextureSample(UVs=AppendVector(A=@n13, B=@n14)) {texture=/Game/Effects/Fort_Effects/Textures/Grad/gradient_healed}.R
  )
  n16 = ScalarParameter() {defaultValue=2, parameterName=FadeSensitivity}
  DepthFade(
   Opacity=Multiply(
    A=Multiply(A=@n1.A, B=@n15)
    B=MaterialFunctionCall("FadeSensitivity (optional)"=@n16) {materialFunction=&FadeBasedOnViewAngle}.Result
   )
   FadeDistance=ScalarParameter() {defaultValue=256, parameterName=Fade Distance}
  )
  n17 = ComponentMask(None=TextureCoordinate()) {g=true}
  n18 = Multiply(
   A=TextureSample(UVs=Panner(Coordinate=@n17, Time=@n3) {speedX=1.6667}) {texture=/Game/Global/Textures/Default/Masks_and_Gradients/T_Grad_Rainbow}.RGB
   B=@n1.RGB
  )
  n19 = ScalarParameter() {defaultValue=10, parameterName=HDR_Intensity}
  MaterialFunctionCall(Emissive=MaterialFunctionCall(LDR Input=Clamp(None=@n18), HDR Tint=@n18, HDR Intensity Multiplier=LinearInterpolate(B=@n19, Alpha=@n4.R) {constA=0.1}) {materialFunction=&DeriveHDRfromLDR}.RemappedRange) {materialFunction=&MF_FX_EmissiveAdjustments}
 ]
}
