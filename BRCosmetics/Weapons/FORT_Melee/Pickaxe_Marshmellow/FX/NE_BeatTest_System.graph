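# Niagara test system for the Marshmellow pickaxe that makes lights pulse to music, using average amplitude from the Papaya_Jukebox_NPC parameter collection.
# One emitter bursts two light-only particles, re-placed each update on skeletal mesh bones or sockets by execution index.
# Squared amplitude scales brightness and light radius.
# Time plus amplitude, wrapped to a fraction, samples a rainbow curve, so hue cycles slowly and jumps with the beat.
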NiagaraSystem /BRCosmetics/Weapons/FORT_Melee/Pickaxe_Marshmellow/FX/NE_BeatTest_System
dependencies {
 usedDataInterfaces [
  &NiagaraDataInterfaceVectorCurve
  &NiagaraDataInterfaceColorCurve
  &NiagaraDataInterfaceSkeletalMesh
 ]
 usedDynamicInputs [&VectorFromCurve, &InvertBool]
 usedModules [
  &self:SystemSpawnScript.SystemScriptSource.SystemScriptGraph.NiagaraNodeAssignment_0.SetVariables_#1
  &SystemState
  &self:SystemSpawnScript.SystemScriptSource.SystemScriptGraph.NiagaraNodeAssignment_1.SetVariables_#2
 ]
}
emitters {NE_BeatTest {
  emitterUpdateScript [
   &EmitterState(Life Cycle Mode=::NewEnumerator1, Loop Behavior=::NewEnumerator1, Loop Duration Mode=::NewEnumerator1, UseLoopDelay=false)
   &SpawnBurst_Instantaneous(Spawn Count=2, Use Spawn Probability=false, UseLoopCountLimit=false)
  ]
  particleSpawnScript [&InitializeParticle(
    Color=<rgba 0.5521 0.5521 0.5521 1>
    Sprite Size Mode=::NewEnumerator3
    Uniform Sprite Size=180
    UsePositionOffset=false
    Write Material Random=true
   ) {module=InitializeParticle001}]
  particleUpdateScript [
   &ParticleState(Kill Particles When Lifetime Has Elapsed=false, Let Infinitely Lived Particles Die When Emitter Deactivates=true)
   &SkeletalMeshLocation(
    Bone / Socket Index=&ReturnExecIndex()
    Bone Sampling Mode=::NewEnumerator4
    MaskSpawnGroup=false
    Sampled Position Offset=<0 0 5>
    Skeletal Mesh=$System.Skeletal Mesh
   )
   &ScaleColor(
    Scale Mode=::NewEnumerator1
    Scale RGBA=&MakeVector4_FromLinearColor(LinearColor=&MultiplyLinearColor(
      Linear Color=&MultiplyLinearColor(
       Linear Color=<rgba 3000 3000 3000 1>
       Scale Factor=&MakeVector4_FromFloat(Float=&Power(a=$NPC.Papaya_Jukebox_NPC.AverageAmplitude, b=2))
      )
      Scale Factor=&Vector4FromCurve(
       CurveIndex=&Frac_Float(Float=&Add_Float(
         A=&Multiply_Float(A=$Engine.Time, B=0.1)
         B=&Multiply_Float(A=$NPC.Papaya_Jukebox_NPC.AverageAmplitude, B=2)
        ))
       Vector4Curve=DataInterface() {propertyValues {
         AlphaCurve {keys [<at 0 1>,<at 1 1>]}
         BlueCurve {keys [<at 0 0.1>,<at 0.1075 0.8314>,<at 0.251 1>,<at 0.4148 1>,<at 0.5618 0.1>,<at 0.7094 0.1>,<at 0.83 0.1>,<at 1 0.1>]}
         GreenCurve {keys [<at 0 0.1>,<at 0.1075 0.1>,<at 0.251 0.2326>,<at 0.4148 0.544>,<at 0.5618 1>,<at 0.7094 0.8714>,<at 0.83 0.5428>,<at 1 0.1>]}
         RedCurve {keys [<at 0 1>,<at 0.1075 1>,<at 0.251 0.1>,<at 0.4148 0.1>,<at 0.5618 0.5135>,<at 0.7094 1>,<at 0.83 1>,<at 1 1>]}
        }}
      )
     ))
    ScaleRGBA=true
   )
   &SolveForcesAndVelocity(Clamp Velocity=false, Limit Acceleration=false)
   &CameraOffset(Camera Offset Amount=5)
   /SetVariables(Particles.LightRadius=&Multiply_Float(A=44, B=&Power(a=$NPC.Papaya_Jukebox_NPC.AverageAmplitude, b=2))) {isSetParameters=true}
  ]
  rendererClasses [&NiagaraLightRendererProperties]
  renderers [{rendererClass=&NiagaraLightRendererProperties, rendererIndex=0}]
 }}
summary {
 systemName=NE_BeatTest_System
 emitters [{emitterName=NE_BeatTest, rendererClasses [&NiagaraLightRendererProperties]}]
 userVariables [User.bisFootHidden {type=NiagaraBool}]
}
